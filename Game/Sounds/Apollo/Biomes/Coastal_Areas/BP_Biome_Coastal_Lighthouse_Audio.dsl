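; Lighthouse audio Blueprint for the coastal biome that sounds a horn when the storm starts shrinking.
; On BeginPlay, where cosmetic events can run, it listens to stateful game phase step events, both standard and Creative storm shield, through the game state event router.
; Both feed one handler, which plays the horn unless it is already playing or the lighthouse is flagged dead.
; EndPlay stops both listeners.
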(event EventEndPlay
  (params (EndPlayReason :type EEndPlayReason%20Enum :default "Destroyed"))
  (bind _gamephasestepupdatedasynctask (Variables|Default|GetGamePhaseStepUpdatedAsyncTask)
    (:then
      (GameplayEventRouter|StopListeningtoStatefulEvent _gamephasestepupdatedasynctask))
    (:else
      _))
  (bind _creativestormshieldgamephasestepupdatedasynctask (Variables|Default|GetCreativeStormShieldGamePhaseStepUpdatedAsyncTask)
    (:then
      (GameplayEventRouter|StopListeningtoStatefulEvent _creativestormshieldgamephasestepupdatedasynctask))
    (:else)))

(event EventBeginPlay
  (Networking|CanExecuteCosmeticEvents
    (:True
      (bind _returnvalue (GameplayEventRouter|GetEventRouter "/Script/FortniteGame.FortGameState" self))
      (bind (_asynctaskproxy _payload) (GameplayEventRouter|StartListeningToStatefulEvent _returnvalue "/Script/FortniteGame.FortGamePhaseStepUpdatedEvent")
        (:then
          (Variables|Default|SetGamePhaseStepUpdatedAsyncTask _asynctaskproxy)
          (bind (_asynctaskproxy_2 _payload_2) (GameplayEventRouter|StartListeningToStatefulEvent _returnvalue "/Script/FortniteGame.FortCreativeStormShieldGamePhaseStepUpdatedEvent")
            (:then
              (Variables|Default|SetCreativeStormShieldGamePhaseStepUpdatedAsyncTask _asynctaskproxy_2))
            (:OnEventReceived
              _)
            (:OnSavedEventStateExists
              _))
          (CallFunction|OnGamePhaseStepChanged (.GamePhaseStep (Utilities|Struct|BreakFortCreativeStormShieldGamePhaseStepUpdatedEvent _payload_2))))
        (:OnEventReceived
          _)
        (:OnSavedEventStateExists
          _))
      (CallFunction|OnGamePhaseStepChanged (Utilities|Struct|BreakFortGamePhaseStepUpdatedEvent _payload)))))

(event Custom|On%20Game%20Phase%20Step%20Changed
  (params (GamePhaseStep :type EAthenaGamePhaseStep%20Enum))
  (bind _hornaudio (Variables|Default|GetHornAudio))
  (switch Utilities|FlowControl|Switch|SwitchonEAthenaGamePhaseStep GamePhaseStep
    (:StormShrinking
      (if (Audio|Components|Audio|IsPlaying _hornaudio)
        _
        (elif (not (Variables|Default|GetbIsLighthouseDead))
          (Audio|Components|Audio|Play _hornaudio 0.0))))))

(fn ConstructionScript)
