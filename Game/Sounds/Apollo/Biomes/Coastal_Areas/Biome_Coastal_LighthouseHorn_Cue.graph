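# Sound cue for the lighthouse foghorn ambience in Apollo's coastal biome.
# It plays a single horn wave after a fixed two-second delay.
# It routes to the outdoor ambient sound class and uses a dedicated lighthouse horn tile attenuation with a very long audible range.
# Virtualization is disabled and the volume is trimmed slightly below full.
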SoundCue /Game/Sounds/Apollo/Biomes/Coastal_Areas/Biome_Coastal_LighthouseHorn_Cue
root=Delay_0
nodes [
 {id=Delay_0, children [WavePlayer_1], properties {delayMax=2, delayMin=2}}
 {id=WavePlayer_1, properties {soundWaveAssetPtr=&/Biome_Coastal_LighthouseHorn}}
]
properties {
 attenuationSettings=/Game/Sounds/Attenuation/Fort_Tile_Ambient_LighthouseHornApollo_Att
 maxDistance=65000
 soundClassObject=&Fort_Ambient_Outdoors_Athena
 virtualizationMode=Disabled
 volumeMultiplier=0.752
 preEffectBusSends [{}]
}
