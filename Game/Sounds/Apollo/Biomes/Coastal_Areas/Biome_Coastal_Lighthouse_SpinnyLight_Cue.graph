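# Sound cue for the looping ambient sound of a lighthouse's spinning light in the Apollo coastal biome.
# A single looping wave player plays the spinning-light loop directly as the root, at reduced volume.
# It uses a dedicated lighthouse attenuation asset and the outdoor ambient sound class, and keeps playing when silent through virtualization.
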SoundCue /Game/Sounds/Apollo/Biomes/Coastal_Areas/Biome_Coastal_Lighthouse_SpinnyLight_Cue
root=WavePlayer_0
nodes [{
  id=WavePlayer_0
  properties {bLooping=true, soundWaveAssetPtr=&/Biome_Coastal_Lighthouse_SpinningLight_Loop}
 }]
properties {
 attenuationSettings=/Game/Sounds/Attenuation/Fort_TileAmbient_Lighthouse_SpinnyLight_Att
 maxDistance=5500
 soundClassObject=&Fort_Ambient_Outdoors_Athena
 virtualizationMode=PlayWhenSilent
 volumeMultiplier=0.501
}
