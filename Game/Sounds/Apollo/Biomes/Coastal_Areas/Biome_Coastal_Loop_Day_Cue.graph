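# Looping daytime ambience cue for the coastal biome, played at low volume through the outdoor ambient sound class and ambient attenuation settings.
# A quality-level node at the root selects between two looping wave players.
# The first two quality slots share the Biome_Coastal_Day wave, while the last two share the Biome_Coastal_Day_ST variant.
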SoundCue /Game/Sounds/Apollo/Biomes/Coastal_Areas/Biome_Coastal_Loop_Day_Cue
root=QualityLevel_0
nodes [
 {id=QualityLevel_0, children [WavePlayer_1,WavePlayer_1,WavePlayer_2,WavePlayer_2]}
 {id=WavePlayer_1, properties {bLooping=true, soundWaveAssetPtr=&/Biome_Coastal_Day}}
 {id=WavePlayer_2, properties {bLooping=true, soundWaveAssetPtr=&/Biome_Coastal_Day_ST}}
]
properties {
 attenuationSettings=&Fort_Ambient_Athena_Att
 maxDistance=2097152
 soundClassObject=&Fort_Ambient_Outdoors_Athena
 volumeMultiplier=0.13
 preEffectBusSends [{}]
}
