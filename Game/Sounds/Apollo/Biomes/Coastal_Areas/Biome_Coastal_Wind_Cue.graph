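# Sound cue for coastal biome wind one-shots, classed as outdoor ambient audio.
# A root Random node picks among five equally weighted wind wave players.
# Playback uses shared tile-ambient one-shot attenuation at reduced volume, with a concurrency override limiting the cue to one instance at a time.
# The cue is excluded from random-node branch culling.
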SoundCue /Game/Sounds/Apollo/Biomes/Coastal_Areas/Biome_Coastal_Wind_Cue
root=Random_0
nodes [
 {id=Random_0, children [WavePlayer_1,WavePlayer_2,WavePlayer_3,WavePlayer_4,WavePlayer_5], properties {weights [1,1,1,1,1]}}
 {id=WavePlayer_1, properties {soundWaveAssetPtr=&/Biome_Coastal_Wind_OneShot_05}}
 {id=WavePlayer_2, properties {soundWaveAssetPtr=&/Biome_Coastal_Wind_OneShot_01}}
 {id=WavePlayer_3, properties {soundWaveAssetPtr=&/Biome_Coastal_Wind_OneShot_02}}
 {id=WavePlayer_4, properties {soundWaveAssetPtr=&/Biome_Coastal_Wind_OneShot_03}}
 {id=WavePlayer_5, properties {soundWaveAssetPtr=&/Biome_Coastal_Wind_OneShot_04}}
]
properties {
 attenuationSettings=/Game/Sounds/Attenuation/Fort_TileAmbient_OneShot_Att
 bExcludeFromRandomNodeBranchCulling=true
 bOverrideConcurrency=true
 maxDistance=8000
 soundClassObject=&Fort_Ambient_Outdoors_Athena
 volumeMultiplier=0.601
 concurrencyOverrides {maxCount=1}
 preEffectBusSends [{}]
}
